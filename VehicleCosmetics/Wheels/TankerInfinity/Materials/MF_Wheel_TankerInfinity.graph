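# Material function for the TankerInfinity wheel cosmetic, blending rim and tire attributes over MF_WheelParent_Simple by vertex color red.
# The rim tints desaturated diffuse with Rim_ColorBase and a fake cubemap reflection borrowed from DuelingDragonsSE.
# Sine time drives MF_Infinity_OffsetFunction on both FX textures for animated infinity-effect emissive in the rim color, plus a pulsing rim mask.
# Tire normals add tiled matte detail, omitted at Low and Medium quality.
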MaterialFunction /VehicleCosmetics/Wheels/TankerInfinity/Materials/MF_Wheel_TankerInfinity
graph {nodes [
  n1 = TextureSampleParameter2D() {
   group=1_InputTextures
   parameterName=Rim_Normal
   samplerType=SAMPLERTYPE_Normal
   texture=/VehicleCosmetics/Wheels/TankerInfinity/Textures/T_Rim_TankerInfinity_N
  }
  ComponentMask(None=@n1.RGB) {b=true, g=true, r=true}
  n2 = MaterialFunctionCall() {materialFunction=&MF_WheelParent_Simple}
  n3 = TextureSampleParameter2D() {group=0_SpecialsEffect, parameterName=Rim_Diffuse, texture=/VehicleCosmetics/Wheels/TankerInfinity/Textures/T_Rim_TankerInfinity_BC}
  n4 = ConstantBiasScale(None=Desaturation(None=@n3.RGB, Fraction=Constant() {r=1})) {bias=0, scale=2}
  n5 = VectorParameter() {parameterName=Rim_ColorBase, defaultValue=<rgba 0.1 0.59 1 1>}
  n6 = TextureSample(
   UVs=Transform(None=ReflectionVectorWS())
   Tex=TextureObject() {samplerType=SAMPLERTYPE_LinearColor, texture=/VehicleCosmetics/Wheels/DuelingDragonsSE/Materials/TC_DuelingDragonsSE_GrassCube}
  ) {texture=/VehicleCosmetics/Wheels/DuelingDragonsSE/Textures/T_Rim_DuelingDragonsSE_RGB}
  n7 = ConstantBiasScale(None=Saturate(None=ConstantBiasScale(None=@n6.R) {bias=-0.5, scale=2})) {bias=0.25, scale=1.75}
  n8 = Multiply(
   A=Multiply(A=Saturate(None=@n4), B=@n5.RGB)
   B=Add(A=Multiply(A=@n7, B=@n6.R), B=ScalarParameter() {defaultValue=0.2, group=0_SpecialsEffect, parameterName=ReflectionBrightness})
  )
  n9 = Add(A=@n6.RGB, B=@n6.RGB)
  n10 = TextureObjectParameter() {
   group=0_SpecialsEffect
   parameterName=_FX
   samplerType=SAMPLERTYPE_LinearColor
   texture=/VehicleCosmetics/Wheels/TankerInfinity/Textures/T_Rim_TankerInfinity_FX
  }
  n11 = TextureSample(Tex=@n10) {samplerType=SAMPLERTYPE_LinearColor, texture=&T_Rim_TankerInfinity_FX}
  n12 = LinearInterpolate(B=Add(A=@n8, B=Multiply(A=@n8, B=@n9)), Alpha=@n11.B)
  n13 = Sine(None=Time()) {period=2}
  n14 = Multiply(A=ScalarParameter() {defaultValue=0.1, group=0_SpecialsEffect, parameterName=BumpOffsetAmount}, B=Multiply(A=@n13) {constB=0.5})
  n15 = ScalarParameter() {defaultValue=0.009, group=0_SpecialsEffect, parameterName=GradientDarkness}
  n16 = Multiply(A=ComponentMask(None=MaterialFunctionCall!(Texture=@n10, Offset=@n14, RotateUVs, GradientDarkness=@n15, GradientFalloff, UVs, Opt_LowQualityOffset, Opt_MediumQualityOffset) {materialFunction=/VehicleCosmetics/SharedMaterials/MaterialFunctions/MF_Infinity_OffsetFunction}.Result) {r=true}) {constB=0.75}
  n17 = Saturate(None=@n13)
  n18 = Max(A=LinearInterpolate(A=Max(A=@n16, B=@n11.R), Alpha=@n11.B) {constB=0}, B=Multiply(A=@n11.G, B=Add(B=@n17) {constA=0.05}))
  n19 = Multiply(A=@n5.RGB, B=Multiply(A=ScalarParameter() {defaultValue=2, group=0_SpecialsEffect, parameterName=EmissiveIntensity}, B=Saturate(None=@n18)))
  n20 = VectorParameter() {parameterName=Tire_ColorBase}
  n21 = TextureObjectParameter() {
   group=0_SpecialsEffect
   parameterName=Tire_FX
   samplerType=SAMPLERTYPE_LinearColor
   texture=/VehicleCosmetics/Wheels/TankerInfinity/Textures/T_Tire_TankerInfinity_RGB
  }
  n22 = Time()
  n23 = Multiply(
   A=ScalarParameter() {defaultValue=0.1, group=0_SpecialsEffect, parameterName=BumpOffsetAmount}
   B=Multiply(A=Sine(None=@n22) {period=2}) {constB=0.5}
  )
  n24 = ScalarParameter() {defaultValue=0.009, group=0_SpecialsEffect, parameterName=GradientDarkness}
  n25 = Multiply(A=ComponentMask(None=MaterialFunctionCall(Texture=@n21, Offset=@n23, GradientDarkness=@n24) {materialFunction=&MF_Infinity_OffsetFunction}.Result) {b=true}) {constB=0.75}
  n26 = TextureSample(Tex=@n21) {samplerType=SAMPLERTYPE_LinearColor, texture=&T_Rim_TankerInfinity_FX}
  n27 = Max(A=LinearInterpolate(A=Max(A=@n25, B=@n26.B), Alpha=@n26.R) {constB=0}) {constB=0}
  n28 = Multiply(A=@n5.RGB, B=Multiply(A=ScalarParameter() {defaultValue=2, group=0_SpecialsEffect, parameterName=EmissiveIntensity}, B=Saturate(None=@n27)))
  n29 = TextureSampleParameter2D() {
   group=1_InputTextures
   parameterName=Tire_Normal
   samplerType=SAMPLERTYPE_Normal
   texture=/VehicleCosmetics/Wheels/TankerInfinity/Textures/T_Tire_TankerInfinity_N
  }
  n30 = Multiply(A=@n29.RGB, B=Constant3Vector() {constant=<rgba 0.625 0.625 1 0>})
  n31 = TextureCoordinate() {uTiling=2, vTiling=2}
  n32 = Add(
   A=@n30
   B=Multiply(
    A=TextureSample(UVs=@n31) {samplerType=SAMPLERTYPE_Normal, texture=/VehicleCosmetics/SharedTextures/T_Matte_N}.RGB
    B=Constant3Vector() {constant=<rgba 0.375 0.375 0 0>}
   )
  )
  n33 = QualitySwitch(Default=@n32, Low=@n30, High=@n32, Medium=@n30, Epic=@n32)
  FunctionOutput(None=BlendMaterialAttributes(
    A=SetMaterialAttributes!(MaterialAttributes=@n2.MaterialAttributes, BaseColor=@n12, EmissiveColor=@n19, Normal=@n1.RGB) {attributeSetTypes [#1, #2, #3]}
    B=SetMaterialAttributes(MaterialAttributes=@n2.MaterialAttributes, BaseColor=@n20.RGB, EmissiveColor=@n28, Normal=@n33) {attributeSetTypes [#1, #2, #3]}
    Alpha=VertexColor().R
   ))
 ]}
